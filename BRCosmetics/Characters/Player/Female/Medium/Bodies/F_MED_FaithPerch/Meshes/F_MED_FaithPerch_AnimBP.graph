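# Anim Blueprint for the FaithPerch female medium body cosmetic, adding secondary motion to its jacket and bag.
# It copies the attached parent mesh's pose into a Control Rig fed by emote and skydive flags.
# Four chained spring-driven Anim Dynamics nodes then simulate the jacket and bag bones in root-relative space.
# Each update sets Gravity Override from PocketBobble's parameter asset, though the dynamics nodes show zero-vector gravity inputs.
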AnimBlueprint /BRCosmetics/Characters/Player/Female/Medium/Bodies/F_MED_FaithPerch/Meshes/F_MED_FaithPerch_AnimBP
target_skeleton=&/F_MED_FaithPerch_Skeleton
graphs [
 AnimGraph {nodes [
   n1 = Control Rig!(Source=Copy Pose From Mesh() {bUseAttachedParent=true}, Alpha, Emotes:bool=Get bIsPlayingEmote(), Skydive:bool=Get bIsSkyDiving()) -> (Pose) {controlRigAssetReference {controlRigAsset=&/F_MED_FaithPerch_CtrlRig}}
   n2 = Anim Dynamics - Bone: dyn_jacket!(ComponentPose=Local To Component(LocalPose=@n1), Alpha=1, GravityOverride:Vector="0,0,0") -> (Pose) {
    angularDampingOverride=0.9999
    angularSpringConstant=150
    bAngularSpring=true
    bLinearSpring=true
    bOverrideAngularDamping=true
    bOverrideLinearDamping=true
    bUseGravityOverride=true
    linearDampingOverride=0.9999
    linearSpringConstant=201
    simulationSpace=RootRelative
    alphaScaleBiasClamp {scale=0.5}
    boundBone {boneName=dyn_jacket}
    physicsBodyDefinitions [{
      boundBone {boneName=dyn_jacket}
      boxExtents=<10 15 10>
      constraintSetup {angularLimitsMax=<10 10 10>, angularLimitsMin=<-10 -10 -10>, angularTarget=<1 0 0>, linearAxesMax=<4 3 3>, linearAxesMin=<0 -3 -3>}
      localJointOffset=<0.01 0 0>
     }]
   }
   n3 = Anim Dynamics - Bone: dyn_jacket_f_l!(ComponentPose=@n2, Alpha=1, GravityOverride:Vector="0,0,0") -> (Pose) {
    angularDampingOverride=0.9999
    angularSpringConstant=150
    bAngularSpring=true
    bLinearSpring=true
    bOverrideAngularDamping=true
    bOverrideLinearDamping=true
    bUseGravityOverride=true
    linearDampingOverride=0.9999
    linearSpringConstant=150
    simulationSpace=RootRelative
    alphaScaleBiasClamp {scale=0.5}
    boundBone {boneName=dyn_jacket_f_l}
    physicsBodyDefinitions [{
      boundBone {boneName=dyn_jacket_f_l}
      boxExtents=<10 10 15>
      constraintSetup {angularLimitsMax=<10 10 10>, angularLimitsMin=<-10 -10 -10>, angularTarget=<1 0 0>, linearAxesMax=<2 3 3>, linearAxesMin=<0 -3 -3>}
      localJointOffset=<0.03 0 0>
     }]
   }
   n4 = Anim Dynamics - Bone: dyn_jacket_f_r!(ComponentPose=@n3, Alpha=1, GravityOverride:Vector="0,0,0") -> (Pose) {
    angularDampingOverride=0.9999
    angularSpringConstant=150
    bAngularSpring=true
    bLinearSpring=true
    bOverrideAngularDamping=true
    bOverrideLinearDamping=true
    bUseGravityOverride=true
    linearDampingOverride=0.9999
    linearSpringConstant=150
    simulationSpace=RootRelative
    alphaScaleBiasClamp {scale=0.5}
    boundBone {boneName=dyn_jacket_f_r}
    physicsBodyDefinitions [{
      boundBone {boneName=dyn_jacket_f_r}
      boxExtents=<10 10 15>
      constraintSetup {angularLimitsMax=<10 10 10>, angularLimitsMin=<-10 -10 -10>, angularTarget=<1 0 0>, linearAxesMax=<2 3 3>, linearAxesMin=<0 -3 -3>}
      localJointOffset=<0.06 0 0>
     }]
   }
   n5 = Anim Dynamics - Bone: dyn_bag!(ComponentPose=@n4, Alpha=1, GravityOverride:Vector="0,0,0") -> (Pose) {
    angularDampingOverride=0.9999
    angularSpringConstant=300
    bAngularSpring=true
    bOverrideAngularDamping=true
    bOverrideLinearDamping=true
    bUseGravityOverride=true
    linearDampingOverride=0.9999
    linearSpringConstant=150
    simulationSpace=RootRelative
    alphaScaleBiasClamp {scale=0.5}
    boundBone {boneName=dyn_bag}
    physicsBodyDefinitions [{
      boundBone {boneName=dyn_bag}
      boxExtents=<10 10 15>
      constraintSetup {angularLimitsMax=<10 3 0>, angularLimitsMin=<-10 -3 -10>, angularTarget=<1 0 0>}
      localJointOffset=<0.1 0 0>
     }]
   }
   Output Pose(Result=Component To Local(ComponentPose=@n5)) {layerGroup=None}
  ]}
 EventGraph {nodes [
   Event BlueprintUpdateAnimation()
   > n1 = "Gravity Override State Function (Parameter Asset) (Nativized)"(
    CharacterPartAnimInstance=Self-Reference()
    GameTimeInSeconds=GetGameTimeInSeconds()
    Parameters=/BRCosmetics/Characters/Player/Female/Medium/Bodies/F_MED_PocketBobble/Meshes/F_MED_PocketBobble_Parameters
   )
   > Set Gravity Override(Gravity Override=@n1)
  ]}
]
variables [Gravity Override {type=Vector}]
