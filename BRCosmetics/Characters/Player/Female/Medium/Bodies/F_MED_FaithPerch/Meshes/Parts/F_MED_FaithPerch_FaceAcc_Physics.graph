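# Physics asset for the FaithPerch female character's face accessory mesh, providing capsule bodies and constraints for hair physics.
# Capsules on the upper spine, clavicles, neck and head form the body volume, joined by angularly free constraints.
# Nine hair chains hang from the head, each with a free-rotating deform root feeding swing-and-twist-limited dyn bones.
# Twist stays tight, swing generally loosens down each chain, and the front side strands are longest.
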PhysicsAsset /BRCosmetics/Characters/Player/Female/Medium/Bodies/F_MED_FaithPerch/Meshes/Parts/F_MED_FaithPerch_FaceAcc_Physics
bodies {
 clavicle_l [{
   shapeType=Capsule
   length=14.213
   radius=5.5923
   shapeName=clavicle_l_capsule
   center=<6.9788 0 -1.302>
   rotation=<rot 78.75 -0.0005 179.9995>
  }]
 clavicle_r [{
   shapeType=Capsule
   length=14.213
   radius=5.5923
   shapeName=clavicle_r_capsule
   center=<-6.9788 0 1.302>
   rotation=<rot 78.75 -0.0005 -0.0005>
  }]
 deform_hair_b [{shapeType=Capsule, length=7, radius=4, shapeName=deform_hair_b_capsule, center=<2.1234 -3.0214 0>, rotation=<rot 90 0 0>}]
 deform_hair_b_l [{shapeType=Capsule, length=7, radius=4, shapeName=deform_hair_b_l_capsule, center=<2.1234 -3.0214 0>, rotation=<rot 90 0 0>}]
 deform_hair_b_r [{shapeType=Capsule, length=7, radius=4, shapeName=deform_hair_b_r_capsule, center=<2.1234 -3.0214 0>, rotation=<rot 90 0 0>}]
 deform_hair_f [{shapeType=Capsule, length=7, radius=4, shapeName=deform_hair_f_capsule, center=<2.1234 -3.0214 0>, rotation=<rot 90 0 0>}]
 deform_hair_f_1 [{shapeType=Capsule, length=7, radius=4, shapeName=deform_hair_f_1_capsule, center=<2.1234 -3.0214 0>, rotation=<rot 90 0 0>}]
 deform_hair_f_l [{shapeType=Capsule, length=8.064, radius=2.936, shapeName=deform_hair_f_l_capsule, center=<2.1234 -2.4387 0>, rotation=<rot 90 0 0>}]
 deform_hair_f_r [{shapeType=Capsule, length=7, radius=4, shapeName=deform_hair_f_r_capsule, center=<2.1234 -3.0214 0>, rotation=<rot 90 0 0>}]
 deform_hair_r [{shapeType=Capsule, length=7, radius=4, shapeName=deform_hair_r_capsule, center=<2.1234 -3.0214 0>, rotation=<rot 90 0 0>}]
 deoform_hair_l [{shapeType=Capsule, length=7, radius=4, shapeName=deoform_hair_l_capsule, center=<2.1234 -3.0214 0>, rotation=<rot 90 0 0>}]
 dyn_hair_1_b [{shapeType=Capsule, length=9.7965, radius=4, shapeName=dyn_hair_1_b_capsule, center=<2.1234 -3.0214 0>}]
 dyn_hair_1_f [{shapeType=Capsule, length=6.3918, radius=4.6082, shapeName=dyn_hair_1_f_capsule, center=<2.1234 -3.0214 0>, rotation=<rot 90 0 0>}]
 dyn_hair_1_l [{shapeType=Capsule, length=9.7965, radius=4, shapeName=dyn_hair_1_l_capsule, center=<2.1234 -3.0214 0>}]
 dyn_hair_1_r [{shapeType=Capsule, length=9.7965, radius=4, shapeName=dyn_hair_1_r_capsule, center=<2.1234 -3.0214 0>}]
 dyn_hair_b [{shapeType=Capsule, length=7, radius=4, shapeName=dyn_hair_b_capsule, center=<2.1234 -3.0214 0>, rotation=<rot 90 0 0>}]
 dyn_hair_b_1_l [{shapeType=Capsule, length=9.7965, radius=4, shapeName=dyn_hair_b_1_l_capsule, center=<2.1234 -3.0214 0>}]
 dyn_hair_b_1_r [{shapeType=Capsule, length=9.7965, radius=4, shapeName=dyn_hair_b_1_r_capsule, center=<2.1234 -3.0214 0>}]
 dyn_hair_b_l [{shapeType=Capsule, length=7, radius=4, shapeName=dyn_hair_b_l_capsule, center=<2.1234 -3.0214 0>, rotation=<rot 90 0 0>}]
 dyn_hair_b_r [{shapeType=Capsule, length=7, radius=4, shapeName=dyn_hair_b_r_capsule, center=<2.1234 -3.0214 0>, rotation=<rot 90 0 0>}]
 dyn_hair_f [{shapeType=Capsule, length=7, radius=4, shapeName=dyn_hair_f_capsule, center=<2.1234 -3.0214 0>, rotation=<rot 90 0 0>}]
 dyn_hair_f_1 [{shapeType=Capsule, length=7, radius=4, shapeName=dyn_hair_f_1_capsule, center=<2.1234 -3.0214 0>, rotation=<rot 90 0 0>}]
 dyn_hair_f_1_l [{shapeType=Capsule, length=8.064, radius=2.936, shapeName=dyn_hair_f_1_l_capsule, center=<3.8971 -2.4387 0>, rotation=<rot 90 0 0>}]
 dyn_hair_f_1_r [{shapeType=Capsule, length=7.9627, radius=3.0373, shapeName=dyn_hair_f_1_r_capsule, center=<3.3084 -3.0214 0>, rotation=<rot 90 0 0>}]
 dyn_hair_f_2_l [{shapeType=Capsule, length=8.7039, radius=2.2961, shapeName=dyn_hair_f_2_l_capsule, center=<2.4193 -1.5032 0>, rotation=<rot 90 0 0>}]
 dyn_hair_f_2_r [{
   shapeType=Capsule
   length=7.7461
   radius=1.7121
   shapeName=dyn_hair_f_2_r_capsule
   center=<3.9959 -0.5105 0>
   rotation=<rot 81.562 -0.0002 -5.6251>
  }]
 dyn_hair_f_3_l [{shapeType=Capsule, length=6.8171, radius=1.5455, shapeName=dyn_hair_f_3_l_capsule, center=<3.6188 0 0>, rotation=<rot 90 0 0>}]
 dyn_hair_f_4 [{shapeType=Capsule, length=7, radius=4, shapeName=dyn_hair_f_4_capsule, center=<2.1234 -3.0214 0>, rotation=<rot 90 0 0>}]
 dyn_hair_f_l [{shapeType=Capsule, length=8.1176, radius=2.8824, shapeName=dyn_hair_f_l_capsule, center=<4.1009 -3.0214 0>, rotation=<rot 90 0 0>}]
 dyn_hair_f_r [{shapeType=Capsule, length=7.0448, radius=3.058, shapeName=dyn_hair_f_r_capsule, center=<3.5065 -3.0214 0>, rotation=<rot 90 0 0>}]
 dyn_hair_l [{shapeType=Capsule, length=7, radius=4, shapeName=dyn_hair_l_capsule, center=<2.1234 -3.0214 0>, rotation=<rot 90 0 0>}]
 dyn_hair_r [{shapeType=Capsule, length=7, radius=4, shapeName=dyn_hair_r_capsule, center=<2.1234 -3.0214 0>, rotation=<rot 90 0 0>}]
 head [
  {shapeType=Capsule, length=12.8128, radius=6.7804, center=<1.9677 1.2436 0>, rotation=<rot 90 -34.7294 0>}
  {shapeType=Capsule, length=1.3816, radius=8.3469, shapeName=head_capsule, center=<7.2289 3.0069 0>, rotation=<rot 90 0 0>}
  {shapeType=Capsule, length=7.2575, radius=3.8896, shapeName=head_capsule, center=<4.0612 0.6225 1.9114>}
 ]
 neck_01 [{shapeType=Capsule, length=5, radius=5, shapeName=neck_01_capsule, center=<2.622 0 0>, rotation=<rot 90 0 0>}]
 spine_04 [{shapeType=Capsule, length=12.4496, radius=11.8615, shapeName=spine_04_capsule, rotation=<rot 0 -90 0>}]
 spine_05 [{shapeType=Capsule, length=14.214, radius=8.0555, shapeName=spine_05_capsule, center=<1.4686 -0.9952 0>}]
}
constraints [
 {
  bone1Name=deform_hair_b
  bone2Name=head
  swing1LimitDegrees=45
  swing1Motion=Free
  swing2LimitDegrees=45
  swing2Motion=Free
  twistLimitDegrees=45
  twistMotion=Free
 }
 {bone1Name=dyn_hair_b, bone2Name=deform_hair_b, swing1LimitDegrees=10, swing2LimitDegrees=10, twistLimitDegrees=5}
 {bone1Name=dyn_hair_1_b, bone2Name=dyn_hair_b, swing1LimitDegrees=25, swing2LimitDegrees=15, twistLimitDegrees=5}
 {
  bone1Name=deform_hair_b_l
  bone2Name=head
  swing1LimitDegrees=45
  swing1Motion=Free
  swing2LimitDegrees=45
  swing2Motion=Free
  twistLimitDegrees=45
  twistMotion=Free
 }
 {bone1Name=dyn_hair_b_l, bone2Name=deform_hair_b_l, swing1LimitDegrees=10, swing2LimitDegrees=10, twistLimitDegrees=5}
 {bone1Name=dyn_hair_b_1_l, bone2Name=dyn_hair_b_l, swing1LimitDegrees=25, swing2LimitDegrees=15, twistLimitDegrees=5}
 {
  bone1Name=deform_hair_b_r
  bone2Name=head
  swing1LimitDegrees=45
  swing1Motion=Free
  swing2LimitDegrees=45
  swing2Motion=Free
  twistLimitDegrees=45
  twistMotion=Free
 }
 {bone1Name=dyn_hair_b_r, bone2Name=deform_hair_b_r, swing1LimitDegrees=10, swing2LimitDegrees=10, twistLimitDegrees=5}
 {bone1Name=dyn_hair_b_1_r, bone2Name=dyn_hair_b_r, swing1LimitDegrees=25, swing2LimitDegrees=15, twistLimitDegrees=5}
 {
  bone1Name=deform_hair_f
  bone2Name=head
  swing1LimitDegrees=45
  swing1Motion=Free
  swing2LimitDegrees=45
  swing2Motion=Free
  twistLimitDegrees=45
  twistMotion=Free
 }
 {bone1Name=dyn_hair_f, bone2Name=deform_hair_f, swing1LimitDegrees=5, swing2LimitDegrees=10, twistLimitDegrees=5}
 {bone1Name=dyn_hair_1_f, bone2Name=dyn_hair_f, swing1LimitDegrees=25, swing2LimitDegrees=15, twistLimitDegrees=5}
 {
  bone1Name=deform_hair_f_1
  bone2Name=head
  swing1LimitDegrees=45
  swing1Motion=Free
  swing2LimitDegrees=45
  swing2Motion=Free
  twistLimitDegrees=45
  twistMotion=Free
 }
 {bone1Name=dyn_hair_f_1, bone2Name=deform_hair_f_1, swing1LimitDegrees=10, swing2LimitDegrees=10, twistLimitDegrees=5}
 {bone1Name=dyn_hair_f_4, bone2Name=dyn_hair_f_1, swing1LimitDegrees=25, swing2LimitDegrees=15, twistLimitDegrees=5}
 {
  bone1Name=deform_hair_f_l
  bone2Name=head
  swing1LimitDegrees=45
  swing1Motion=Free
  swing2LimitDegrees=45
  swing2Motion=Free
  twistLimitDegrees=45
  twistMotion=Free
 }
 {bone1Name=dyn_hair_f_l, bone2Name=deform_hair_f_l, swing1LimitDegrees=10, swing2LimitDegrees=1, twistLimitDegrees=5}
 {bone1Name=dyn_hair_f_1_l, bone2Name=dyn_hair_f_l, swing1LimitDegrees=25, swing2LimitDegrees=15, twistLimitDegrees=5}
 {bone1Name=dyn_hair_f_2_l, bone2Name=dyn_hair_f_1_l, swing1LimitDegrees=25, swing2LimitDegrees=25, twistLimitDegrees=5}
 {bone1Name=dyn_hair_f_3_l, bone2Name=dyn_hair_f_2_l, swing1LimitDegrees=25, swing2LimitDegrees=25, twistLimitDegrees=5}
 {
  bone1Name=deform_hair_f_r
  bone2Name=head
  swing1LimitDegrees=45
  swing1Motion=Free
  swing2LimitDegrees=45
  swing2Motion=Free
  twistLimitDegrees=45
  twistMotion=Free
 }
 {bone1Name=dyn_hair_f_r, bone2Name=deform_hair_f_r, swing1LimitDegrees=10, swing2LimitDegrees=10, twistLimitDegrees=5}
 {bone1Name=dyn_hair_f_1_r, bone2Name=dyn_hair_f_r, swing1LimitDegrees=25, swing2LimitDegrees=15, twistLimitDegrees=5}
 {bone1Name=dyn_hair_f_2_r, bone2Name=dyn_hair_f_1_r, swing1LimitDegrees=25, swing2LimitDegrees=15, twistLimitDegrees=5}
 {
  bone1Name=deform_hair_r
  bone2Name=head
  swing1LimitDegrees=45
  swing1Motion=Free
  swing2LimitDegrees=45
  swing2Motion=Free
  twistLimitDegrees=45
  twistMotion=Free
 }
 {bone1Name=dyn_hair_r, bone2Name=deform_hair_r, swing1LimitDegrees=10, swing2LimitDegrees=10, twistLimitDegrees=5}
 {bone1Name=dyn_hair_1_r, bone2Name=dyn_hair_r, swing1LimitDegrees=25, swing2LimitDegrees=15, twistLimitDegrees=5}
 {
  bone1Name=deoform_hair_l
  bone2Name=head
  swing1LimitDegrees=45
  swing1Motion=Free
  swing2LimitDegrees=45
  swing2Motion=Free
  twistLimitDegrees=45
  twistMotion=Free
 }
 {bone1Name=dyn_hair_l, bone2Name=deoform_hair_l, swing1LimitDegrees=10, swing2LimitDegrees=10, twistLimitDegrees=5}
 {bone1Name=dyn_hair_1_l, bone2Name=dyn_hair_l, swing1LimitDegrees=25, swing2LimitDegrees=15, twistLimitDegrees=5}
 {
  bone1Name=head
  bone2Name=neck_01
  swing1LimitDegrees=45
  swing1Motion=Free
  swing2LimitDegrees=45
  swing2Motion=Free
  twistLimitDegrees=45
  twistMotion=Free
 }
 {
  bone1Name=clavicle_l
  bone2Name=spine_05
  swing1LimitDegrees=45
  swing1Motion=Free
  swing2LimitDegrees=45
  swing2Motion=Free
  twistLimitDegrees=45
  twistMotion=Free
 }
 {
  bone1Name=clavicle_r
  bone2Name=spine_05
  swing1LimitDegrees=45
  swing1Motion=Free
  swing2LimitDegrees=45
  swing2Motion=Free
  twistLimitDegrees=45
  twistMotion=Free
 }
 {
  bone1Name=neck_01
  bone2Name=spine_05
  swing1LimitDegrees=45
  swing1Motion=Free
  swing2LimitDegrees=45
  swing2Motion=Free
  twistLimitDegrees=45
  twistMotion=Free
 }
 {
  bone1Name=spine_05
  bone2Name=spine_04
  swing1LimitDegrees=45
  swing1Motion=Free
  swing2LimitDegrees=45
  swing2Motion=Free
  twistLimitDegrees=45
  twistMotion=Free
 }
]
