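# Animation Blueprint for the FaithPerch female medium body's face accessory, adding physics-driven secondary motion.
# It copies the pose from the attached parent mesh, then runs a RigidBody node with its own physics asset, simulated in head-bone space.
# Each update, a nativized gravity override function stores gravity and rigid-body settings in variables, using the HumorShale face accessory's parameter asset rather than a FaithPerch one.
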AnimBlueprint /BRCosmetics/Characters/Player/Female/Medium/Bodies/F_MED_FaithPerch/Meshes/Parts/F_MED_FaithPerch_FaceAcc_AnimBP
target_skeleton=&/F_MED_FaithPerch_FaceAcc_Skeleton
graphs [
 AnimGraph {nodes [
   n1 = RigidBody!(
    ComponentPose=Local To Component(LocalPose=Copy Pose From Mesh() {bUseAttachedParent=true})
    Alpha
    OverrideWorldGravity:Vector="0,0,0"
    SimSpaceSettings:Sim Space Settings Structure=(WorldAlpha=0.5,VelocityScaleZ=0.4,DampingAlpha=1,MaxLinearVelocity=5000,MaxAngularVelocity=5000,MaxLinearAcceleration=2500,MaxAngularAcceleration=500,ExternalLinearDragV=(X=1,Y=1,Z=1),ExternalLinearVelocity=(X=0,Y=0,Z=0),ExternalAngularVelocity=(X=0,Y=0,Z=0))
   ) -> (Pose) {
    bClampLinearTranslationLimitToRefPose=true
    bOverrideWorldGravity=true
    overridePhysicsAsset=&/F_MED_FaithPerch_FaceAcc_Physics
    simulationSpace=BaseBoneSpace
    baseBoneRef {boneName=head}
    simSpaceSettings {
     maxAngularAcceleration=500
     maxAngularVelocity=5000
     maxLinearAcceleration=2500
     maxLinearVelocity=5000
     velocityScaleZ=0.4
     worldAlpha=0.5
     externalLinearDragV=<1 1 1>
    }
   }
   Output Pose(Result=Component To Local(ComponentPose=@n1)) {layerGroup=None}
  ]}
 EventGraph {nodes [
   Event BlueprintUpdateAnimation()
   > n1 = "Gravity Override State Function (Parameter Asset) (Nativized)"(
    CharacterPartAnimInstance=Self-Reference()
    GameTimeInSeconds=GetGameTimeInSeconds()
    Parameters=/BRCosmetics/Characters/Player/Female/Medium/Bodies/F_MED_HumorShale/Meshes/Parts/F_MED_HumorShale_FaceAcc_Parameters
   )
   > Set Gravity Override(Gravity Override=@n1)
   > Set RBAN Settings Out(RBAN Settings Out=@n1.OutRigidBodyAnimNodeParameters)
  ]}
]
variables [
 Gravity Override {type=Vector}
 RBAN Settings Out {type=FortRigidBodyAnimNodeParameters}
]
